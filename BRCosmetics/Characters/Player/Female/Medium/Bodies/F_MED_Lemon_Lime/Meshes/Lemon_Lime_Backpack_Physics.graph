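# Physics asset for the Lemon Lime backpack cosmetic on the medium female character, covering dynamic bandanna, collar, strap, jacket and belt bones.
# Capsules cover spine, neck and head, while box-body chains hang from small base bodies, mostly joined to spine_04 with free angular motion.
# Segments use narrow angular limits, with twist locked on back panels, whose lower segments are cross-linked sideways.
# Belts attach asymmetrically: left free on the front panel, right locked to its base.
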PhysicsAsset /BRCosmetics/Characters/Player/Female/Medium/Bodies/F_MED_Lemon_Lime/Meshes/Lemon_Lime_Backpack_Physics
bodies {
 dyn_bandanna [{shapeType=Box, extentX=1.01, extentY=1.01, extentZ=1.01, rotation=<rot 0 -90 0>}]
 dyn_bandannaA [{shapeType=Box, extentX=0.9988, extentY=2.2656, extentZ=4.0081, center=<2.0305 0.0796 0.0547>, rotation=<rot 0 -90.6853 -85.6085>}]
 dyn_bandannaB [{shapeType=Box, extentX=1.2465, extentY=2.0245, extentZ=4.143, center=<1.9691 0.3985 -0.0108>, rotation=<rot 0 84.7493 -81.4289>}]
 dyn_collarA_01_Base_l [{shapeType=Box, extentX=1.01, extentY=1.01, extentZ=1.01, rotation=<rot 0 -90 0>}]
 dyn_collarA_01_Base_r [{shapeType=Box, extentX=1.01, extentY=1.01, extentZ=1.01, rotation=<rot 0 -90 0>}]
 dyn_collarA_01_l [{shapeType=Box, extentX=3.0455, extentY=5.6509, extentZ=7.922, center=<2.3127 0.0088 -0.0885>, rotation=<rot -5.6149 -89.9997 2.2026>}]
 dyn_collarA_01_r [{shapeType=Box, extentX=3.0455, extentY=5.6509, extentZ=7.922, center=<2.3127 0.0088 -0.0885>, rotation=<rot 82.824 -72.0828 17.8327>}]
 dyn_collarA_02_l [{shapeType=Box, extentX=2.5564, extentY=5.6509, extentZ=8.2057, center=<2.3272 -0.0279 0.2852>, rotation=<rot -5.6149 -89.9997 2.2026>}]
 dyn_collarA_02_r [{shapeType=Box, extentX=2.2705, extentY=5.6509, extentZ=8.2057, center=<2.3272 -0.0041 0.0432>, rotation=<rot -84.3849 90.0001 -177.7973>}]
 dyn_collarA_03_l [{shapeType=Box, extentX=3.0455, extentY=8.4591, extentZ=6.071, center=<3.2194 0.0123 -0.1232>, rotation=<rot 1.1101 -89.7414 2.1925>}]
 dyn_collarA_03_r [{shapeType=Box, extentX=2.1924, extentY=3.0219, extentZ=10.2512, center=<1.0398 0.5122 -0.1335>, rotation=<rot 88.8799 90.2488 179.3767>}]
 dyn_collarB_01_Base_l [{shapeType=Box, extentX=1.01, extentY=1.01, extentZ=1.01, rotation=<rot 0 -90 0>}]
 dyn_collarB_01_Base_r [{shapeType=Box, extentX=1.01, extentY=1.01, extentZ=1.01, rotation=<rot 0 -90 0>}]
 dyn_collarB_01_l [{shapeType=Box, extentX=1.8411, extentY=2.9675, extentZ=7.922, center=<0.7349 0.0028 -0.0281>, rotation=<rot 9.9844 -89.3979 2.2258>}]
 dyn_collarB_01_r [{shapeType=Box, extentX=1.8411, extentY=2.9675, extentZ=7.922, center=<0.8392 0.2592 0.0537>, rotation=<rot 71.3728 -83.2618 6.8789>}]
 dyn_collarB_02_l [{shapeType=Box, extentX=2.0236, extentY=4.8743, extentZ=7.8035, center=<2.3175 0.0088 -0.0887>, rotation=<rot 10.6807 -89.5312 -3.8759>}]
 dyn_collarB_02_r [{shapeType=Box, extentX=2.0236, extentY=4.8743, extentZ=7.8035, center=<2.3328 0.068 -0.3358>, rotation=<rot 80.8977 -113.3829 -13.9942>}]
 dyn_collarBack_01 [{shapeType=Box, extentX=2.5564, extentY=5.6509, extentZ=8.2057, center=<2.3142 -0.0333 -0.0542>, rotation=<rot 0.8981 -89.7495 0.767>}]
 dyn_collarBack_01_Base [{shapeType=Box, extentX=1.01, extentY=1.01, extentZ=1.01, rotation=<rot 0 -90 0>}]
 dyn_collarBack_02 [{shapeType=Box, extentX=2.5564, extentY=4.8228, extentZ=8.2057, center=<1.7687 -0.03 0.3065>, rotation=<rot -1.3392 -89.8352 2.1927>}]
 dyn_collarD_01_Base_l [{shapeType=Box, extentX=1.01, extentY=1.01, extentZ=1.01, rotation=<rot 0 -90 0>}]
 dyn_collarD_01_Base_r [{shapeType=Box, extentX=1.01, extentY=1.01, extentZ=1.01, rotation=<rot 0 -90 0>}]
 dyn_collarD_01_l [{shapeType=Box, extentX=2.5564, extentY=4.8444, extentZ=8.2057, center=<1.8867 0.4253 0.2349>, rotation=<rot 8.3647 -89.4616 2.2157>}]
 dyn_collarD_01_r [{shapeType=Box, extentX=2.5564, extentY=5.1309, extentZ=8.2057, center=<1.781 -0.0268 0.2729>, rotation=<rot -5.6149 -89.9998 -1.2936>}]
 dyn_collarD_02_l [{shapeType=Box, extentX=2.5564, extentY=4.1896, extentZ=8.2057, center=<1.2697 -0.0345 0.3041>, rotation=<rot 6.6807 -89.5261 -10.1808>}]
 dyn_collarD_02_r [{shapeType=Box, extentX=2.5564, extentY=4.5538, extentZ=8.2057, center=<1.5046 0.0104 -0.1045>, rotation=<rot -5.6149 -89.9997 2.2026>}]
 dyn_collarStrap_01_l [{shapeType=Box, extentX=3.0455, extentY=2.6751, extentZ=2.6334, center=<1.1696 0.0047 -0.0448>, rotation=<rot 1.1101 -89.7414 2.1925>}]
 dyn_collarStrap_01_r [{shapeType=Box, extentX=3.0455, extentY=2.6751, extentZ=2.6334, center=<1.1696 0.0047 -0.0448>, rotation=<rot 1.1101 -89.7414 2.1925>}]
 dyn_collarStrap_02_l [{shapeType=Box, extentX=1.1786, extentY=2.6751, extentZ=2.6334, center=<1.1696 0.0047 -0.0448>, rotation=<rot 1.1101 -89.7414 2.1925>}]
 dyn_collarStrap_02_r [{shapeType=Box, extentX=1.1786, extentY=2.6751, extentZ=2.6334, center=<1.1696 0.0047 -0.0448>, rotation=<rot 91.1101 -89.7414 2.1925>}]
 dyn_jacketBack_01 [{shapeType=Box, extentX=5.8879, extentY=2.5619, extentZ=6.9846, center=<2.0881 0.0214 -0.0951>}]
 dyn_jacketBack_01_Base [{shapeType=Box, extentX=1.01, extentY=1.01, extentZ=1.01, rotation=<rot 0 -90 0>}]
 dyn_jacketBack_01_Base_l [{shapeType=Box, extentX=1.01, extentY=1.01, extentZ=1.01, rotation=<rot 0 -90 0>}]
 dyn_jacketBack_01_Base_r [{shapeType=Box, extentX=1.01, extentY=1.01, extentZ=1.01, rotation=<rot 0 -90 0>}]
 dyn_jacketBack_01_l [{shapeType=Box, extentX=5.8421, extentY=2.2241, extentZ=6.6092, center=<2.4824 -0.0902 -0.2447>}]
 dyn_jacketBack_01_r [{shapeType=Box, extentX=5.8421, extentY=2.1632, extentZ=6.6092, center=<2.5186 -0.0046 0>}]
 dyn_jacketBack_02 [{shapeType=Box, extentX=9.603, extentY=2.1384, extentZ=8.4305, center=<4.5364 -0.5226 -0.3736>, rotation=<rot 0 -6.1308 0>}]
 dyn_jacketBack_02_l [{shapeType=Box, extentX=9.8158, extentY=2.1384, extentZ=6.6092, center=<4.1702 -0.1317 -0.2935>, rotation=<rot 0 0.2348 0>}]
 dyn_jacketBack_02_r [{shapeType=Box, extentX=9.8241, extentY=2.1384, extentZ=6.6092, center=<4.3285 -0.0493 -0.0193>}]
 dyn_jacketBelt01_Base_l [{shapeType=Box, extentX=1.01, extentY=1.01, extentZ=1.01, rotation=<rot 0 -90 0>}]
 dyn_jacketBelt01_Base_r [{shapeType=Box, extentX=1.01, extentY=1.01, extentZ=1.01, rotation=<rot 0 -90 0>}]
 dyn_jacketBelt01_l [{shapeType=Box, extentX=2.9101, extentY=0.544, extentZ=1.9599, center=<1.1827 -0.0924 -0.0047>, rotation=<rot 0 2.9907 -72.0588>}]
 dyn_jacketBelt01_r [{shapeType=Box, extentX=0.973, extentY=2.1314, extentZ=2.922, center=<1.0145 -0.092 -0.0749>, rotation=<rot 14.3636 90.1442 -94.0519>}]
 dyn_jacketBelt02_l [{shapeType=Box, extentX=2.505, extentY=1.01, extentZ=2.1674, center=<-1.1016 -0.0627 0.0201>, rotation=<rot -2.6803 179.725 -74.3402>}]
 dyn_jacketBelt02_r [{shapeType=Box, extentX=0.973, extentY=2.1314, extentZ=2.922, center=<1.1887 0.0482 0.0164>, rotation=<rot 0 84.8433 -92.5225>}]
 dyn_jacketBelt03_l [{shapeType=Box, extentX=2.4052, extentY=1.01, extentZ=2.5156, center=<-1.0969 -0.0476 -0.004>, rotation=<rot 0 178.1415 -75.6108>}]
 dyn_jacketBelt03_r [{shapeType=Box, extentX=0.973, extentY=2.1314, extentZ=2.922, center=<1.1391 0.0527 0.009>, rotation=<rot 0 84.8435 -92.7675>}]
 dyn_jacketBelt04_l [{shapeType=Box, extentX=1.7383, extentY=1.01, extentZ=2.791, center=<-0.6666 -0.0261 -0.013>, rotation=<rot 0 -179.8116 -75.9397>}]
 dyn_jacketBelt04_r [{shapeType=Box, extentX=0.973, extentY=2.1314, extentZ=1.9894, center=<0.7197 0.0905 0.0165>, rotation=<rot 0 84.8435 -90.2423>}]
 dyn_jacketFront01_Base_l [{shapeType=Box, extentX=1.01, extentY=1.01, extentZ=1.01, rotation=<rot 0 -90 0>}]
 dyn_jacketFront01_Base_r [{shapeType=Box, extentX=1.01, extentY=1.01, extentZ=1.01, rotation=<rot 0 -90 0>}]
 dyn_jacketFront01_l [{shapeType=Box, extentX=7.9502, extentY=0.881, extentZ=8.9619, center=<5.2372 0.0003 -0.2306>}]
 dyn_jacketFront01_r [{shapeType=Box, extentX=8.3254, extentY=1.283, extentZ=7.6966, center=<4.0428 0.4443 -0.7866>, rotation=<rot 0 7.4961 0>}]
 dyn_jacketSideA01_Base_l [{shapeType=Box, extentX=1.01, extentY=1.01, extentZ=1.01, rotation=<rot 0 -90 0>}]
 dyn_jacketSideA01_Base_r [{shapeType=Box, extentX=1.01, extentY=1.01, extentZ=1.01, rotation=<rot 0 -90 0>}]
 dyn_jacketSideA01_l [{shapeType=Box, extentX=6.2047, extentY=1.5167, extentZ=6.6092, center=<2.7509 0.2233 0.005>, rotation=<rot -12.1424 6.2143 -1.312>}]
 dyn_jacketSideA01_r [{shapeType=Box, extentX=6.2047, extentY=1.2746, extentZ=6.6092, center=<2.7741 0.0001 0>, rotation=<rot 13.5459 0 0>}]
 dyn_jacketSideB_03_Base_l [{shapeType=Box, extentX=1.01, extentY=1.01, extentZ=1.01, rotation=<rot 0 -90 0>}]
 dyn_jacketSideB_03_Base_r [{shapeType=Box, extentX=1.01, extentY=1.01, extentZ=1.01, rotation=<rot 0 -90 0>}]
 dyn_jacketSideB_03_l [{shapeType=Box, extentX=6.2047, extentY=1.511, extentZ=6.6092, center=<2.8099 0.4526 0.0002>, rotation=<rot -0.2046 9.1311 -0.0596>}]
 dyn_jacketSideB_03_r [{shapeType=Box, extentX=6.2047, extentY=0.881, extentZ=6.6092, center=<2.7741 0.0001 0>}]
 head [{shapeType=Capsule, length=4.4368, radius=5.5126, center=<0 2.2187 0>, rotation=<rot 90 0 -16.001>}]
 neck_01 [{shapeType=Capsule, length=7.1283, radius=5.0099, center=<7.063 0.0001 0>, rotation=<rot 90 0 0>}]
 spine_03 [
  {shapeType=Capsule, length=8.983, radius=6.8018, center=<0.6194 1.8307 0>, rotation=<rot 0 -90 0>}
  {shapeType=Capsule, length=4.3112, radius=6.9578, center=<1.5996 5.4513 0>}
 ]
 spine_04 [
  {shapeType=Capsule, length=4.4062, radius=10.2796, center=<0.4859 3.5498 0.116>, rotation=<rot 90 -90 -79.5623>}
  {shapeType=Capsule, length=0.118, radius=9.8776, center=<2.1539 4.3981 0>}
  {shapeType=Capsule, length=4.7776, radius=5.6107, center=<0.0002 -0.5226 5.6627>, rotation=<rot 90 0 0>}
  {shapeType=Capsule, length=4.7776, radius=5.6107, center=<0.0002 -0.5226 -5.6627>, rotation=<rot 90 0 0>}
 ]
}
constraints [
 {bone1Name=dyn_bandannaB, bone2Name=dyn_bandanna, swing1LimitDegrees=20, swing2LimitDegrees=20, twistLimitDegrees=10}
 {bone1Name=dyn_bandannaA, bone2Name=dyn_bandanna, swing1LimitDegrees=20, swing2LimitDegrees=20, twistLimitDegrees=10}
 {
  bone1Name=dyn_jacketBack_01_l
  bone2Name=dyn_jacketBack_01_Base_l
  swing1LimitDegrees=4
  swing2LimitDegrees=16
  swing2Motion=Locked
  twistLimitDegrees=8
  twistMotion=Locked
 }
 {
  bone1Name=dyn_jacketBack_02_l
  bone2Name=dyn_jacketBack_01_l
  swing1LimitDegrees=4
  swing2LimitDegrees=2
  twistLimitDegrees=8
  twistMotion=Locked
 }
 {
  bone1Name=dyn_jacketBack_01
  bone2Name=dyn_jacketBack_01_Base
  swing1LimitDegrees=4
  swing2LimitDegrees=8
  swing2Motion=Locked
  twistLimitDegrees=8
  twistMotion=Locked
 }
 {
  bone1Name=dyn_jacketBack_02
  bone2Name=dyn_jacketBack_01
  swing1LimitDegrees=4
  swing2LimitDegrees=2
  twistLimitDegrees=4
  twistMotion=Locked
 }
 {
  bone1Name=dyn_jacketBack_01_r
  bone2Name=dyn_jacketBack_01_Base_r
  swing1LimitDegrees=4
  swing2LimitDegrees=16
  swing2Motion=Locked
  twistLimitDegrees=8
  twistMotion=Locked
 }
 {
  bone1Name=dyn_jacketBack_02_r
  bone2Name=dyn_jacketBack_01_r
  swing1LimitDegrees=4
  swing2LimitDegrees=2
  twistLimitDegrees=8
  twistMotion=Locked
 }
 {
  bone1Name=dyn_bandanna
  bone2Name=spine_04
  swing1LimitDegrees=45
  swing1Motion=Free
  swing2LimitDegrees=45
  swing2Motion=Free
  twistLimitDegrees=45
  twistMotion=Free
 }
 {
  bone1Name=dyn_jacketBack_01_Base_l
  bone2Name=spine_04
  swing1LimitDegrees=45
  swing1Motion=Free
  swing2LimitDegrees=45
  swing2Motion=Free
  twistLimitDegrees=45
  twistMotion=Free
 }
 {
  bone1Name=dyn_jacketBack_01_Base
  bone2Name=spine_04
  swing1LimitDegrees=45
  swing1Motion=Free
  swing2LimitDegrees=45
  swing2Motion=Free
  twistLimitDegrees=45
  twistMotion=Free
 }
 {
  bone1Name=dyn_jacketBack_01_Base_r
  bone2Name=spine_04
  swing1LimitDegrees=45
  swing1Motion=Free
  swing2LimitDegrees=45
  swing2Motion=Free
  twistLimitDegrees=45
  twistMotion=Free
 }
 {
  bone1Name=spine_04
  bone2Name=spine_03
  swing1LimitDegrees=45
  swing1Motion=Free
  swing2LimitDegrees=45
  swing2Motion=Free
  twistLimitDegrees=45
  twistMotion=Free
 }
 {
  bone1Name=dyn_jacketSideB_03_Base_l
  bone2Name=spine_04
  swing1LimitDegrees=45
  swing1Motion=Free
  swing2LimitDegrees=45
  swing2Motion=Free
  twistLimitDegrees=45
  twistMotion=Free
 }
 {bone1Name=dyn_jacketSideB_03_l, bone2Name=dyn_jacketSideB_03_Base_l, swing1LimitDegrees=20, swing2LimitDegrees=16, twistLimitDegrees=11}
 {
  bone1Name=dyn_jacketSideB_03_Base_r
  bone2Name=spine_04
  swing1LimitDegrees=45
  swing1Motion=Free
  swing2LimitDegrees=45
  swing2Motion=Free
  twistLimitDegrees=45
  twistMotion=Free
 }
 {bone1Name=dyn_jacketSideB_03_r, bone2Name=dyn_jacketSideB_03_Base_r, swing1LimitDegrees=20, swing2LimitDegrees=16, twistLimitDegrees=11}
 {bone1Name=dyn_jacketSideA01_l, bone2Name=dyn_jacketSideA01_Base_l, swing1LimitDegrees=20, swing2LimitDegrees=16, twistLimitDegrees=11}
 {
  bone1Name=dyn_jacketSideA01_Base_l
  bone2Name=spine_04
  swing1LimitDegrees=45
  swing1Motion=Free
  swing2LimitDegrees=45
  swing2Motion=Free
  twistLimitDegrees=45
  twistMotion=Free
 }
 {bone1Name=dyn_jacketSideA01_r, bone2Name=dyn_jacketSideA01_Base_r, swing1LimitDegrees=20, swing2LimitDegrees=16, twistLimitDegrees=11}
 {
  bone1Name=dyn_jacketSideA01_Base_r
  bone2Name=spine_04
  swing1LimitDegrees=45
  swing1Motion=Free
  swing2LimitDegrees=45
  swing2Motion=Free
  twistLimitDegrees=45
  twistMotion=Free
 }
 {
  bone1Name=dyn_jacketFront01_Base_r
  bone2Name=spine_04
  swing1LimitDegrees=45
  swing1Motion=Free
  swing2LimitDegrees=45
  swing2Motion=Free
  twistLimitDegrees=45
  twistMotion=Free
 }
 {bone1Name=dyn_jacketFront01_r, bone2Name=dyn_jacketFront01_Base_r, swing1LimitDegrees=30, swing2LimitDegrees=30, twistLimitDegrees=12}
 {
  bone1Name=dyn_jacketFront01_Base_l
  bone2Name=spine_04
  swing1LimitDegrees=45
  swing1Motion=Free
  swing2LimitDegrees=45
  swing2Motion=Free
  twistLimitDegrees=45
  twistMotion=Free
 }
 {bone1Name=dyn_jacketFront01_l, bone2Name=dyn_jacketFront01_Base_l, swing1LimitDegrees=20, swing2LimitDegrees=20, twistLimitDegrees=12}
 {
  bone1Name=dyn_jacketBelt01_Base_l
  bone2Name=dyn_jacketFront01_l
  swing1LimitDegrees=45
  swing1Motion=Free
  swing2LimitDegrees=45
  swing2Motion=Free
  twistLimitDegrees=45
  twistMotion=Free
 }
 {bone1Name=dyn_jacketBelt01_l, bone2Name=dyn_jacketBelt01_Base_l, swing1LimitDegrees=20, swing2LimitDegrees=16, twistLimitDegrees=8}
 {bone1Name=dyn_jacketBelt02_l, bone2Name=dyn_jacketBelt01_l, swing1LimitDegrees=20, swing2LimitDegrees=16, twistLimitDegrees=8}
 {bone1Name=dyn_jacketBelt03_l, bone2Name=dyn_jacketBelt02_l, swing1LimitDegrees=20, swing2LimitDegrees=16, twistLimitDegrees=8}
 {bone1Name=dyn_jacketBelt04_l, bone2Name=dyn_jacketBelt03_l, swing1LimitDegrees=20, swing2LimitDegrees=16, twistLimitDegrees=8}
 {bone1Name=dyn_collarStrap_02_l, bone2Name=dyn_collarStrap_01_l, swing1LimitDegrees=20, swing2LimitDegrees=10, twistLimitDegrees=8}
 {bone1Name=dyn_collarStrap_01_l, bone2Name=dyn_collarA_02_l, swing1LimitDegrees=20, swing2LimitDegrees=10, twistLimitDegrees=8}
 {bone1Name=dyn_collarA_03_l, bone2Name=dyn_collarA_02_l, swing1LimitDegrees=30, swing2LimitDegrees=10, twistLimitDegrees=10}
 {bone1Name=dyn_collarA_02_l, bone2Name=dyn_collarA_01_l, swing1LimitDegrees=30, swing2LimitDegrees=10, twistLimitDegrees=10}
 {bone1Name=dyn_collarA_01_l, bone2Name=dyn_collarA_01_Base_l, swing1LimitDegrees=30, swing2LimitDegrees=10, twistLimitDegrees=10}
 {bone1Name=dyn_collarA_01_Base_l, bone2Name=spine_04, swing1LimitDegrees=30, swing2LimitDegrees=10, twistLimitDegrees=10}
 {
  bone1Name=dyn_jacketBelt01_Base_r
  bone2Name=dyn_jacketFront01_Base_r
  swing1LimitDegrees=45
  swing1Motion=Locked
  swing2LimitDegrees=45
  swing2Motion=Locked
  twistLimitDegrees=45
  twistMotion=Locked
 }
 {bone1Name=dyn_jacketBelt01_r, bone2Name=dyn_jacketBelt01_Base_r, swing1LimitDegrees=20, swing2LimitDegrees=16, twistLimitDegrees=8}
 {bone1Name=dyn_jacketBelt02_r, bone2Name=dyn_jacketBelt01_r, swing1LimitDegrees=20, swing2LimitDegrees=16, twistLimitDegrees=8}
 {bone1Name=dyn_jacketBelt03_r, bone2Name=dyn_jacketBelt02_r, swing1LimitDegrees=20, swing2LimitDegrees=16, twistLimitDegrees=8}
 {bone1Name=dyn_jacketBelt04_r, bone2Name=dyn_jacketBelt03_r, swing1LimitDegrees=20, swing2LimitDegrees=16, twistLimitDegrees=8}
 {
  bone1Name=dyn_collarB_01_Base_l
  bone2Name=spine_04
  swing1LimitDegrees=45
  swing1Motion=Free
  swing2LimitDegrees=45
  swing2Motion=Free
  twistLimitDegrees=45
  twistMotion=Free
 }
 {bone1Name=dyn_collarB_01_l, bone2Name=dyn_collarB_01_Base_l, swing1LimitDegrees=20, swing2LimitDegrees=10, twistLimitDegrees=8}
 {bone1Name=dyn_collarB_02_l, bone2Name=dyn_collarB_01_l, swing1LimitDegrees=20, swing2LimitDegrees=10, twistLimitDegrees=8}
 {
  bone1Name=dyn_collarD_01_Base_l
  bone2Name=spine_04
  swing1LimitDegrees=45
  swing1Motion=Free
  swing2LimitDegrees=45
  swing2Motion=Free
  twistLimitDegrees=45
  twistMotion=Free
 }
 {bone1Name=dyn_collarD_01_l, bone2Name=dyn_collarD_01_Base_l, swing1LimitDegrees=20, swing2LimitDegrees=10, twistLimitDegrees=8}
 {bone1Name=dyn_collarD_02_l, bone2Name=dyn_collarD_01_l, swing1LimitDegrees=20, swing2LimitDegrees=10, twistLimitDegrees=8}
 {
  bone1Name=dyn_collarBack_01_Base
  bone2Name=spine_04
  swing1LimitDegrees=45
  swing1Motion=Free
  swing2LimitDegrees=45
  swing2Motion=Free
  twistLimitDegrees=45
  twistMotion=Free
 }
 {bone1Name=dyn_collarBack_01, bone2Name=dyn_collarBack_01_Base, swing1LimitDegrees=20, swing2LimitDegrees=10, twistLimitDegrees=8}
 {bone1Name=dyn_collarBack_02, bone2Name=dyn_collarBack_01, swing1LimitDegrees=20, swing2LimitDegrees=10, twistLimitDegrees=8}
 {
  bone1Name=dyn_collarD_01_Base_r
  bone2Name=spine_04
  swing1LimitDegrees=45
  swing1Motion=Free
  swing2LimitDegrees=45
  swing2Motion=Free
  twistLimitDegrees=45
  twistMotion=Free
 }
 {bone1Name=dyn_collarD_01_r, bone2Name=dyn_collarD_01_Base_r, swing1LimitDegrees=20, swing2LimitDegrees=10, twistLimitDegrees=8}
 {bone1Name=dyn_collarD_02_r, bone2Name=dyn_collarD_01_r, swing1LimitDegrees=20, swing2LimitDegrees=10, twistLimitDegrees=8}
 {
  bone1Name=dyn_collarB_01_Base_r
  bone2Name=spine_04
  swing1LimitDegrees=45
  swing1Motion=Free
  swing2LimitDegrees=45
  swing2Motion=Free
  twistLimitDegrees=45
  twistMotion=Free
 }
 {bone1Name=dyn_collarB_01_r, bone2Name=dyn_collarB_01_Base_r, swing1LimitDegrees=10, swing2LimitDegrees=20, twistLimitDegrees=8}
 {bone1Name=dyn_collarB_02_r, bone2Name=dyn_collarB_01_r, swing1LimitDegrees=10, swing2LimitDegrees=20, twistLimitDegrees=8}
 {
  bone1Name=dyn_collarA_01_Base_r
  bone2Name=spine_04
  swing1LimitDegrees=45
  swing1Motion=Free
  swing2LimitDegrees=45
  swing2Motion=Free
  twistLimitDegrees=45
  twistMotion=Free
 }
 {bone1Name=dyn_collarA_01_r, bone2Name=dyn_collarA_01_Base_r, swing1LimitDegrees=10, swing2LimitDegrees=30, twistLimitDegrees=10}
 {bone1Name=dyn_collarA_02_r, bone2Name=dyn_collarA_01_r, swing1LimitDegrees=10, swing2LimitDegrees=30, twistLimitDegrees=10}
 {bone1Name=dyn_collarA_03_r, bone2Name=dyn_collarA_02_r, swing1LimitDegrees=10, swing2LimitDegrees=30, twistLimitDegrees=10}
 {bone1Name=dyn_collarStrap_01_r, bone2Name=dyn_collarA_02_r, swing1LimitDegrees=10, swing2LimitDegrees=20, twistLimitDegrees=8}
 {bone1Name=dyn_collarStrap_02_r, bone2Name=dyn_collarStrap_01_r, swing1LimitDegrees=10, swing2LimitDegrees=20, twistLimitDegrees=8}
 {
  bone1Name=neck_01
  bone2Name=spine_04
  swing1LimitDegrees=45
  swing1Motion=Free
  swing2LimitDegrees=45
  swing2Motion=Free
  twistLimitDegrees=45
  twistMotion=Free
 }
 {
  bone1Name=head
  bone2Name=neck_01
  swing1LimitDegrees=45
  swing1Motion=Free
  swing2LimitDegrees=45
  swing2Motion=Free
  twistLimitDegrees=45
  twistMotion=Free
 }
 {
  bone1Name=dyn_jacketBack_02_r
  bone2Name=dyn_jacketBack_02
  swing1LimitDegrees=45
  swing1Motion=Free
  swing2LimitDegrees=45
  swing2Motion=Free
  twistLimitDegrees=45
  twistMotion=Free
 }
 {
  bone1Name=dyn_jacketBack_02_l
  bone2Name=dyn_jacketBack_02
  swing1LimitDegrees=45
  swing1Motion=Free
  swing2LimitDegrees=45
  swing2Motion=Free
  twistLimitDegrees=45
  twistMotion=Free
 }
]
