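# Physics asset for the Lemon_Lime female medium body cosmetic, set up for secondary motion of its bandanna, collar, straps, jacket panels and belts.
# Capsules on the upper spine, neck, head and right clavicle form the body.
# Small base boxes anchor limited swing/twist chains, mostly to spine_04, with right collars A/B on clavicle_r.
# Straps and belts branch from collar and front-jacket bodies, and free constraints cross-link the lower back panels.
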PhysicsAsset /BRCosmetics/Characters/Player/Female/Medium/Bodies/F_MED_Lemon_Lime/Meshes/Lemon_Lime_Physics
bodies {
 clavicle_r [{shapeType=Capsule, length=18.5557, radius=5.795, shapeName=clavicle_r_capsule, center=<-12.0501 -0.0002 1.9283>, rotation=<rot 49.9994 0 0>}]
 dyn_bandanna [{shapeType=Box, extentX=1.01, extentY=1.01, extentZ=1.01, rotation=<rot 0 -90 0>}]
 dyn_bandannaA [{shapeType=Box, extentX=0.9988, extentY=2.2656, extentZ=4.0081, center=<2.0305 0.0796 0.0547>, rotation=<rot 0 -90.6853 -85.6085>}]
 dyn_bandannaB [{shapeType=Box, extentX=1.2465, extentY=2.0245, extentZ=4.143, center=<1.9691 0.3985 -0.0108>, rotation=<rot 0 84.7493 -81.4289>}]
 dyn_collarA_01_Base_l [{shapeType=Box, extentX=1.01, extentY=1.01, extentZ=1.01, rotation=<rot 0 -90 0>}]
 dyn_collarA_01_Base_r [{shapeType=Box, extentX=1.01, extentY=1.01, extentZ=1.01, rotation=<rot 0 -90 0>}]
 dyn_collarA_01_l [{shapeType=Capsule, length=6.5, radius=2, shapeName=dyn_collarA_01_l_capsule, center=<2.4939 0 1.7937>, rotation=<rot 20.0001 0 0>}]
 dyn_collarA_01_r [{shapeType=Capsule, length=6.5, radius=2.25, shapeName=dyn_collarA_01_r_capsule, center=<5.0458 -0.0975 0.5529>, rotation=<rot 0 0 79.998>}]
 dyn_collarA_02_l [{shapeType=Capsule, length=6.5, radius=2, shapeName=dyn_collarA_02_l_capsule, center=<4.4453 0 1.7937>, rotation=<rot 20.0001 0 0>}]
 dyn_collarA_02_r [{shapeType=Capsule, length=6.5, radius=2.25, shapeName=dyn_collarA_02_r_capsule, center=<5.0458 -0.0975 0.5529>, rotation=<rot 0 0 79.998>}]
 dyn_collarA_03_l [{shapeType=Capsule, length=6.5, radius=2, shapeName=dyn_collarA_03_l_capsule, center=<4.7656 0 1.7937>, rotation=<rot 20.0001 0 0>}]
 dyn_collarA_03_r [{shapeType=Capsule, length=6.5, radius=2.25, shapeName=dyn_collarA_03_r_capsule, center=<4.0861 -0.0975 0.5529>, rotation=<rot 0 0 79.9978>}]
 dyn_collarB_01_Base_l [{shapeType=Box, extentX=1.01, extentY=1.01, extentZ=1.01, rotation=<rot 0 -90 0>}]
 dyn_collarB_01_Base_r [{shapeType=Box, extentX=1.01, extentY=1.01, extentZ=1.01, rotation=<rot 0 -90 0>}]
 dyn_collarB_01_l [{shapeType=Capsule, length=6.75, radius=2, shapeName=dyn_collarB_01_l_capsule, center=<3.8807 0 0>}]
 dyn_collarB_01_r [{shapeType=Capsule, length=3.25, radius=2.25, shapeName=dyn_collarB_01_r_capsule, center=<4.1738 -0.0976 0.5531>, rotation=<rot 0 0 79.9991>}]
 dyn_collarB_02_l [{shapeType=Capsule, length=6.75, radius=2, shapeName=dyn_collarB_02_l_capsule, center=<7.141 0 0>}]
 dyn_collarB_02_r [{shapeType=Capsule, length=3.25, radius=2.25, shapeName=dyn_collarB_02_r_capsule, center=<5.5824 -0.8937 0.4127>, rotation=<rot 0 0 79.9977>}]
 dyn_collarBack_01 [{shapeType=Capsule, length=3.5, radius=1.75, shapeName=dyn_collarBack_01_capsule, center=<6.5005 0 0>}]
 dyn_collarBack_01_Base [{shapeType=Box, extentX=1.01, extentY=1.01, extentZ=1.01, rotation=<rot 0 -90 0>}]
 dyn_collarBack_02 [{shapeType=Capsule, length=3.5, radius=1.75, shapeName=dyn_collarBack_02_capsule, center=<6.5005 0 0>}]
 dyn_collarD_01_Base_l [{shapeType=Box, extentX=1.01, extentY=1.01, extentZ=1.01, rotation=<rot 0 -90 0>}]
 dyn_collarD_01_Base_r [{shapeType=Box, extentX=1.01, extentY=1.01, extentZ=1.01, rotation=<rot 0 -90 0>}]
 dyn_collarD_01_l [{shapeType=Capsule, length=3.5, radius=1.75, shapeName=dyn_collarD_01_l_capsule, center=<6.5005 0 0>}]
 dyn_collarD_01_r [{shapeType=Capsule, length=3.5, radius=1.75, shapeName=dyn_collarD_01_r_capsule, center=<6.5005 0 0>}]
 dyn_collarD_02_l [{shapeType=Capsule, length=3.5, radius=1.75, shapeName=dyn_collarD_02_l_capsule, center=<6.5005 0 0>}]
 dyn_collarD_02_r [{shapeType=Capsule, length=3.5, radius=1.75, shapeName=dyn_collarD_02_r_capsule, center=<6.5005 0 0>}]
 dyn_collarStrap_01_l [{shapeType=Box, extentX=3.0455, extentY=2.6751, extentZ=2.6334, center=<1.1696 0.0047 -0.0448>, rotation=<rot 1.1101 -89.7414 2.1925>}]
 dyn_collarStrap_01_r [{shapeType=Box, extentX=3.0455, extentY=2.6751, extentZ=2.6334, center=<1.1696 0.0047 -0.0448>, rotation=<rot 1.1101 -89.7414 2.1925>}]
 dyn_collarStrap_02_l [{shapeType=Box, extentX=1.1786, extentY=2.6751, extentZ=2.6334, center=<1.1696 0.0047 -0.0448>, rotation=<rot 1.1101 -89.7414 2.1925>}]
 dyn_collarStrap_02_r [{shapeType=Box, extentX=1.1786, extentY=2.6751, extentZ=2.6334, center=<1.1696 0.0047 -0.0448>, rotation=<rot 91.1101 -89.7414 2.1925>}]
 dyn_jacketBack_01 [{
   shapeType=Capsule
   length=2.75
   radius=2
   shapeName=dyn_jacketBack_01_capsule
   center=<-0.8259 0.404 -0.0348>
   rotation=<rot 90 1.2659 1.266>
  }]
 dyn_jacketBack_01_Base [{shapeType=Box, extentX=1.01, extentY=1.01, extentZ=1.01, rotation=<rot 0 -90 0>}]
 dyn_jacketBack_01_Base_l [{shapeType=Box, extentX=1.01, extentY=1.01, extentZ=1.01, rotation=<rot 0 -90 0>}]
 dyn_jacketBack_01_Base_r [{shapeType=Box, extentX=1.01, extentY=1.01, extentZ=1.01, rotation=<rot 0 -90 0>}]
 dyn_jacketBack_01_l [{shapeType=Capsule, length=2.75, radius=2, shapeName=dyn_jacketBack_01_l_capsule, center=<1.0962 0.8644 0>, rotation=<rot 90 -0.0002 0.0001>}]
 dyn_jacketBack_01_r [{
   shapeType=Capsule
   length=2.75
   radius=3
   shapeName=dyn_jacketBack_01_r_capsule
   center=<8.9142 4.3796 -1.2961>
   rotation=<rot -10.0013 -179.9998 179.9998>
  }]
 dyn_jacketBack_02 [{
   shapeType=Capsule
   length=2.25
   radius=2
   shapeName=dyn_jacketBack_02_capsule
   center=<0.8651 -0.1902 -0.0348>
   rotation=<rot 90 0.211 0.211>
  }]
 dyn_jacketBack_02_l [{
   shapeType=Capsule
   length=2.25
   radius=1.75
   shapeName=dyn_jacketBack_02_l_capsule
   center=<2.7872 0.2703 0>
   rotation=<rot 90 10.0251 10.0253>
  }]
 dyn_jacketBack_02_r [{
   shapeType=Capsule
   length=2
   radius=2.75
   shapeName=dyn_jacketBack_02_r_capsule
   center=<8.3283 4.4043 -0.7226>
   rotation=<rot -10.0001 179.9991 -179.9999>
  }]
 dyn_jacketBelt01_Base_l [{shapeType=Box, extentX=1.01, extentY=1.01, extentZ=1.01, rotation=<rot 0 -90 0>}]
 dyn_jacketBelt01_Base_r [{shapeType=Box, extentX=1.01, extentY=1.01, extentZ=1.01, rotation=<rot 0 -90 0>}]
 dyn_jacketBelt01_l [{shapeType=Box, extentX=2.9101, extentY=0.544, extentZ=1.9599, center=<1.1827 -0.0924 -0.0047>, rotation=<rot 0 2.9907 -72.0588>}]
 dyn_jacketBelt01_r [{shapeType=Box, extentX=0.973, extentY=2.1314, extentZ=2.922, center=<1.0025 0.0909 -0.0188>, rotation=<rot 18.0496 95.2387 -92.7792>}]
 dyn_jacketBelt02_l [{shapeType=Box, extentX=2.505, extentY=1.01, extentZ=2.1674, center=<-1.1016 -0.0627 0.0201>, rotation=<rot -2.6803 179.725 -74.3402>}]
 dyn_jacketBelt02_r [{shapeType=Box, extentX=0.973, extentY=2.1314, extentZ=2.922, center=<1.1888 0.1828 0.0169>, rotation=<rot 0.225 89.958 -92.5123>}]
 dyn_jacketBelt03_l [{shapeType=Box, extentX=2.4052, extentY=1.01, extentZ=2.5156, center=<-1.0969 -0.0476 -0.004>, rotation=<rot 0 178.1415 -75.6108>}]
 dyn_jacketBelt03_r [{shapeType=Box, extentX=0.973, extentY=2.1314, extentZ=2.922, center=<1.1391 0.0527 0.009>, rotation=<rot 0 84.8435 -92.7675>}]
 dyn_jacketBelt04_l [{shapeType=Box, extentX=1.7383, extentY=1.01, extentZ=2.791, center=<-0.6666 -0.0261 -0.013>, rotation=<rot 0 -179.8116 -75.9397>}]
 dyn_jacketBelt04_r [{shapeType=Box, extentX=0.973, extentY=2.1314, extentZ=1.9894, center=<0.7197 0.0905 0.0165>, rotation=<rot 0 84.8435 -90.2423>}]
 dyn_jacketFront01_Base_l [{shapeType=Box, extentX=1.01, extentY=1.01, extentZ=1.01, rotation=<rot 0 -90 0>}]
 dyn_jacketFront01_Base_r [{shapeType=Box, extentX=1.01, extentY=1.01, extentZ=1.01, rotation=<rot 0 -90 0>}]
 dyn_jacketFront01_l [{shapeType=Capsule, length=3.25, radius=1.75, shapeName=dyn_jacketFront01_l_capsule, center=<6.9808 0 -0.8937>, rotation=<rot 30.0001 0 0>}]
 dyn_jacketFront01_r [{
   shapeType=Capsule
   length=6.5
   radius=1.75
   shapeName=dyn_jacketFront01_r_capsule
   center=<6.5763 -3.4427 1.9534>
   rotation=<rot -20.0002 0 0>
  }]
 dyn_jacketSideA01_Base_l [{shapeType=Box, extentX=1.01, extentY=1.01, extentZ=1.01, rotation=<rot 0 -90 0>}]
 dyn_jacketSideA01_Base_r [{shapeType=Box, extentX=1.01, extentY=1.01, extentZ=1.01, rotation=<rot 0 -90 0>}]
 dyn_jacketSideA01_l [{shapeType=Capsule, length=3, radius=1.75, shapeName=dyn_jacketSideA01_l_capsule, center=<5.2924 0 0>}]
 dyn_jacketSideA01_r [{shapeType=Capsule, length=3, radius=1.75, shapeName=dyn_jacketSideA01_r_capsule, center=<5.2924 0 0>}]
 dyn_jacketSideB_03_Base_l [{shapeType=Box, extentX=1.01, extentY=1.01, extentZ=1.01, rotation=<rot 0 -90 0>}]
 dyn_jacketSideB_03_Base_r [{shapeType=Box, extentX=1.01, extentY=1.01, extentZ=1.01, rotation=<rot 0 -90 0>}]
 dyn_jacketSideB_03_l [{shapeType=Capsule, length=3, radius=1.75, shapeName=dyn_jacketSideB_03_l_capsule, center=<5.2924 0 0>}]
 dyn_jacketSideB_03_r [{shapeType=Capsule, length=3, radius=1.75, shapeName=dyn_jacketSideB_03_r_capsule, center=<5.2924 0 0>}]
 head [{shapeType=Capsule, length=4.4368, radius=5.5126, center=<0 2.2187 0>, rotation=<rot 90 0 -16.001>}]
 neck_01 [
  {shapeType=Capsule, length=7.1283, radius=5.0099, center=<7.063 0.0001 0>, rotation=<rot 90 0 0>}
  {shapeType=Capsule, length=6, radius=6, shapeName=neck_01_capsule, center=<-3.6314 1.1717 0>}
 ]
 spine_03 [
  {shapeType=Capsule, length=8.983, radius=6.8018, center=<0.6194 1.8307 0>, rotation=<rot 0 -90 0>}
  {shapeType=Capsule, length=4.3112, radius=6.9578, center=<1.5996 5.4513 0>}
 ]
 spine_04 [
  {shapeType=Capsule, length=4.4062, radius=10.2796, center=<0.4859 3.5498 0.116>, rotation=<rot 90 -90 -79.5623>}
  {shapeType=Capsule, length=0.5821, radius=9.2956, center=<2.1541 5.7649 0>}
  {shapeType=Capsule, length=4.7776, radius=5.6107, center=<0.0002 -0.5226 5.6627>, rotation=<rot 90 0 0>}
  {shapeType=Capsule, length=4.7776, radius=5.6107, center=<0.0002 -0.5226 -5.6627>, rotation=<rot 90 0 0>}
 ]
}
constraints [
 {bone1Name=dyn_bandannaB, bone2Name=dyn_bandanna, swing1LimitDegrees=20, swing2LimitDegrees=20, twistLimitDegrees=10}
 {bone1Name=dyn_bandannaA, bone2Name=dyn_bandanna, swing1LimitDegrees=20, swing2LimitDegrees=20, twistLimitDegrees=10}
 {bone1Name=dyn_jacketBack_01_l, bone2Name=dyn_jacketBack_01_Base_l, swing1LimitDegrees=20, swing2LimitDegrees=16, twistLimitDegrees=8}
 {bone1Name=dyn_jacketBack_02_l, bone2Name=dyn_jacketBack_01_l, swing1LimitDegrees=20, swing2LimitDegrees=16, twistLimitDegrees=8}
 {bone1Name=dyn_jacketBack_01, bone2Name=dyn_jacketBack_01_Base, swing1LimitDegrees=20, swing2LimitDegrees=8, twistLimitDegrees=8}
 {
  bone1Name=dyn_jacketBack_02
  bone2Name=dyn_jacketBack_01
  swing1LimitDegrees=20
  swing2LimitDegrees=10
  twistLimitDegrees=4
  twistMotion=Locked
 }
 {bone1Name=dyn_jacketBack_01_r, bone2Name=dyn_jacketBack_01_Base_r, swing1LimitDegrees=20, swing2LimitDegrees=16, twistLimitDegrees=8}
 {bone1Name=dyn_jacketBack_02_r, bone2Name=dyn_jacketBack_01_r, swing1LimitDegrees=20, swing2LimitDegrees=16, twistLimitDegrees=8}
 {
  bone1Name=dyn_bandanna
  bone2Name=spine_04
  swing1LimitDegrees=45
  swing1Motion=Free
  swing2LimitDegrees=45
  swing2Motion=Free
  twistLimitDegrees=45
  twistMotion=Free
 }
 {
  bone1Name=dyn_jacketBack_01_Base_l
  bone2Name=spine_04
  swing1LimitDegrees=45
  swing1Motion=Free
  swing2LimitDegrees=45
  swing2Motion=Free
  twistLimitDegrees=45
  twistMotion=Free
 }
 {
  bone1Name=dyn_jacketBack_01_Base
  bone2Name=spine_04
  swing1LimitDegrees=45
  swing1Motion=Free
  swing2LimitDegrees=45
  swing2Motion=Free
  twistLimitDegrees=45
  twistMotion=Free
 }
 {
  bone1Name=dyn_jacketBack_01_Base_r
  bone2Name=spine_04
  swing1LimitDegrees=45
  swing1Motion=Free
  swing2LimitDegrees=45
  swing2Motion=Free
  twistLimitDegrees=45
  twistMotion=Free
 }
 {
  bone1Name=spine_04
  bone2Name=spine_03
  swing1LimitDegrees=45
  swing1Motion=Free
  swing2LimitDegrees=45
  swing2Motion=Free
  twistLimitDegrees=45
  twistMotion=Free
 }
 {
  bone1Name=dyn_jacketSideB_03_Base_l
  bone2Name=spine_04
  swing1LimitDegrees=45
  swing1Motion=Free
  swing2LimitDegrees=45
  swing2Motion=Free
  twistLimitDegrees=45
  twistMotion=Free
 }
 {bone1Name=dyn_jacketSideB_03_l, bone2Name=dyn_jacketSideB_03_Base_l, swing1LimitDegrees=20, swing2LimitDegrees=16, twistLimitDegrees=11}
 {
  bone1Name=dyn_jacketSideB_03_Base_r
  bone2Name=spine_04
  swing1LimitDegrees=45
  swing1Motion=Free
  swing2LimitDegrees=45
  swing2Motion=Free
  twistLimitDegrees=45
  twistMotion=Free
 }
 {bone1Name=dyn_jacketSideB_03_r, bone2Name=dyn_jacketSideB_03_Base_r, swing1LimitDegrees=20, swing2LimitDegrees=16, twistLimitDegrees=11}
 {bone1Name=dyn_jacketSideA01_l, bone2Name=dyn_jacketSideA01_Base_l, swing1LimitDegrees=20, swing2LimitDegrees=16, twistLimitDegrees=11}
 {
  bone1Name=dyn_jacketSideA01_Base_l
  bone2Name=spine_04
  swing1LimitDegrees=45
  swing1Motion=Free
  swing2LimitDegrees=45
  swing2Motion=Free
  twistLimitDegrees=45
  twistMotion=Free
 }
 {bone1Name=dyn_jacketSideA01_r, bone2Name=dyn_jacketSideA01_Base_r, swing1LimitDegrees=20, swing2LimitDegrees=16, twistLimitDegrees=11}
 {
  bone1Name=dyn_jacketSideA01_Base_r
  bone2Name=spine_04
  swing1LimitDegrees=45
  swing1Motion=Free
  swing2LimitDegrees=45
  swing2Motion=Free
  twistLimitDegrees=45
  twistMotion=Free
 }
 {
  bone1Name=dyn_jacketFront01_Base_r
  bone2Name=spine_04
  swing1LimitDegrees=45
  swing1Motion=Free
  swing2LimitDegrees=45
  swing2Motion=Free
  twistLimitDegrees=45
  twistMotion=Free
 }
 {bone1Name=dyn_jacketFront01_r, bone2Name=dyn_jacketFront01_Base_r, swing1LimitDegrees=30, swing2LimitDegrees=30, twistLimitDegrees=12}
 {
  bone1Name=dyn_jacketFront01_Base_l
  bone2Name=spine_04
  swing1LimitDegrees=45
  swing1Motion=Free
  swing2LimitDegrees=45
  swing2Motion=Free
  twistLimitDegrees=45
  twistMotion=Free
 }
 {bone1Name=dyn_jacketFront01_l, bone2Name=dyn_jacketFront01_Base_l, swing1LimitDegrees=20, swing2LimitDegrees=20, twistLimitDegrees=12}
 {
  bone1Name=dyn_jacketBelt01_Base_l
  bone2Name=dyn_jacketFront01_l
  swing1LimitDegrees=45
  swing1Motion=Free
  swing2LimitDegrees=45
  swing2Motion=Free
  twistLimitDegrees=45
  twistMotion=Free
 }
 {bone1Name=dyn_jacketBelt01_l, bone2Name=dyn_jacketBelt01_Base_l, swing1LimitDegrees=20, swing2LimitDegrees=16, twistLimitDegrees=8}
 {bone1Name=dyn_jacketBelt02_l, bone2Name=dyn_jacketBelt01_l, swing1LimitDegrees=20, swing2LimitDegrees=16, twistLimitDegrees=8}
 {bone1Name=dyn_jacketBelt03_l, bone2Name=dyn_jacketBelt02_l, swing1LimitDegrees=20, swing2LimitDegrees=16, twistLimitDegrees=8}
 {bone1Name=dyn_jacketBelt04_l, bone2Name=dyn_jacketBelt03_l, swing1LimitDegrees=20, swing2LimitDegrees=16, twistLimitDegrees=8}
 {bone1Name=dyn_collarStrap_02_l, bone2Name=dyn_collarStrap_01_l, swing1LimitDegrees=20, swing2LimitDegrees=10, twistLimitDegrees=8}
 {bone1Name=dyn_collarStrap_01_l, bone2Name=dyn_collarA_02_l, swing1LimitDegrees=20, swing2LimitDegrees=10, twistLimitDegrees=8}
 {bone1Name=dyn_collarA_03_l, bone2Name=dyn_collarA_02_l, swing1LimitDegrees=30, swing2LimitDegrees=10, twistLimitDegrees=10}
 {bone1Name=dyn_collarA_02_l, bone2Name=dyn_collarA_01_l, swing1LimitDegrees=30, swing2LimitDegrees=10, twistLimitDegrees=10}
 {bone1Name=dyn_collarA_01_l, bone2Name=dyn_collarA_01_Base_l, swing1LimitDegrees=30, swing2LimitDegrees=10, twistLimitDegrees=10}
 {bone1Name=dyn_collarA_01_Base_l, bone2Name=spine_04, swing1LimitDegrees=30, swing2LimitDegrees=10, twistLimitDegrees=10}
 {
  bone1Name=dyn_jacketBelt01_Base_r
  bone2Name=dyn_jacketFront01_Base_r
  swing1LimitDegrees=45
  swing1Motion=Locked
  swing2LimitDegrees=45
  swing2Motion=Locked
  twistLimitDegrees=45
  twistMotion=Locked
 }
 {bone1Name=dyn_jacketBelt01_r, bone2Name=dyn_jacketBelt01_Base_r, swing1LimitDegrees=6, swing2LimitDegrees=6, twistLimitDegrees=4}
 {bone1Name=dyn_jacketBelt02_r, bone2Name=dyn_jacketBelt01_r, swing1LimitDegrees=6, swing2LimitDegrees=6, twistLimitDegrees=4}
 {bone1Name=dyn_jacketBelt03_r, bone2Name=dyn_jacketBelt02_r, swing1LimitDegrees=20, swing2LimitDegrees=16, twistLimitDegrees=8}
 {bone1Name=dyn_jacketBelt04_r, bone2Name=dyn_jacketBelt03_r, swing1LimitDegrees=20, swing2LimitDegrees=16, twistLimitDegrees=8}
 {
  bone1Name=dyn_collarB_01_Base_l
  bone2Name=spine_04
  swing1LimitDegrees=45
  swing1Motion=Free
  swing2LimitDegrees=45
  swing2Motion=Free
  twistLimitDegrees=45
  twistMotion=Free
 }
 {bone1Name=dyn_collarB_01_l, bone2Name=dyn_collarB_01_Base_l, swing1LimitDegrees=20, swing2LimitDegrees=10, twistLimitDegrees=8}
 {bone1Name=dyn_collarB_02_l, bone2Name=dyn_collarB_01_l, swing1LimitDegrees=20, swing2LimitDegrees=10, twistLimitDegrees=8}
 {
  bone1Name=dyn_collarD_01_Base_l
  bone2Name=spine_04
  swing1LimitDegrees=45
  swing1Motion=Free
  swing2LimitDegrees=45
  swing2Motion=Free
  twistLimitDegrees=45
  twistMotion=Free
 }
 {bone1Name=dyn_collarD_01_l, bone2Name=dyn_collarD_01_Base_l, swing1LimitDegrees=20, swing2LimitDegrees=10, twistLimitDegrees=8}
 {bone1Name=dyn_collarD_02_l, bone2Name=dyn_collarD_01_l, swing1LimitDegrees=20, swing2LimitDegrees=10, twistLimitDegrees=8}
 {
  bone1Name=dyn_collarBack_01_Base
  bone2Name=spine_04
  swing1LimitDegrees=45
  swing1Motion=Free
  swing2LimitDegrees=45
  swing2Motion=Free
  twistLimitDegrees=45
  twistMotion=Free
 }
 {bone1Name=dyn_collarBack_01, bone2Name=dyn_collarBack_01_Base, swing1LimitDegrees=20, swing2LimitDegrees=10, twistLimitDegrees=8}
 {bone1Name=dyn_collarBack_02, bone2Name=dyn_collarBack_01, swing1LimitDegrees=20, swing2LimitDegrees=10, twistLimitDegrees=8}
 {
  bone1Name=dyn_collarD_01_Base_r
  bone2Name=spine_04
  swing1LimitDegrees=45
  swing1Motion=Free
  swing2LimitDegrees=45
  swing2Motion=Free
  twistLimitDegrees=45
  twistMotion=Free
 }
 {bone1Name=dyn_collarD_01_r, bone2Name=dyn_collarD_01_Base_r, swing1LimitDegrees=20, swing2LimitDegrees=10, twistLimitDegrees=8}
 {bone1Name=dyn_collarD_02_r, bone2Name=dyn_collarD_01_r, swing1LimitDegrees=20, swing2LimitDegrees=10, twistLimitDegrees=8}
 {
  bone1Name=dyn_collarB_01_Base_r
  bone2Name=clavicle_r
  swing1LimitDegrees=45
  swing1Motion=Free
  swing2LimitDegrees=45
  swing2Motion=Free
  twistLimitDegrees=45
  twistMotion=Free
 }
 {bone1Name=dyn_collarB_01_r, bone2Name=dyn_collarB_01_Base_r, swing1LimitDegrees=10, swing2LimitDegrees=20, twistLimitDegrees=8}
 {bone1Name=dyn_collarB_02_r, bone2Name=dyn_collarB_01_r, swing1LimitDegrees=10, swing2LimitDegrees=20, twistLimitDegrees=8}
 {
  bone1Name=dyn_collarA_01_Base_r
  bone2Name=clavicle_r
  swing1LimitDegrees=45
  swing1Motion=Free
  swing2LimitDegrees=45
  swing2Motion=Free
  twistLimitDegrees=45
  twistMotion=Free
 }
 {bone1Name=dyn_collarA_01_r, bone2Name=dyn_collarA_01_Base_r, swing1LimitDegrees=10, swing2LimitDegrees=30, twistLimitDegrees=10}
 {bone1Name=dyn_collarA_02_r, bone2Name=dyn_collarA_01_r, swing1LimitDegrees=10, swing2LimitDegrees=30, twistLimitDegrees=10}
 {bone1Name=dyn_collarA_03_r, bone2Name=dyn_collarA_02_r, swing1LimitDegrees=10, swing2LimitDegrees=30, twistLimitDegrees=10}
 {bone1Name=dyn_collarStrap_01_r, bone2Name=dyn_collarA_02_r, swing1LimitDegrees=10, swing2LimitDegrees=20, twistLimitDegrees=8}
 {bone1Name=dyn_collarStrap_02_r, bone2Name=dyn_collarStrap_01_r, swing1LimitDegrees=10, swing2LimitDegrees=20, twistLimitDegrees=8}
 {
  bone1Name=neck_01
  bone2Name=spine_04
  swing1LimitDegrees=45
  swing1Motion=Free
  swing2LimitDegrees=45
  swing2Motion=Free
  twistLimitDegrees=45
  twistMotion=Free
 }
 {
  bone1Name=head
  bone2Name=neck_01
  swing1LimitDegrees=45
  swing1Motion=Free
  swing2LimitDegrees=45
  swing2Motion=Free
  twistLimitDegrees=45
  twistMotion=Free
 }
 {
  bone1Name=dyn_jacketBack_02_r
  bone2Name=dyn_jacketBack_02
  swing1LimitDegrees=45
  swing1Motion=Free
  swing2LimitDegrees=45
  swing2Motion=Free
  twistLimitDegrees=45
  twistMotion=Free
 }
 {
  bone1Name=dyn_jacketBack_02_l
  bone2Name=dyn_jacketBack_02
  swing1LimitDegrees=45
  swing1Motion=Free
  swing2LimitDegrees=45
  swing2Motion=Free
  twistLimitDegrees=45
  twistMotion=Free
 }
 {
  bone1Name=clavicle_r
  bone2Name=spine_04
  swing1LimitDegrees=45
  swing1Motion=Free
  swing2LimitDegrees=45
  swing2Motion=Free
  twistLimitDegrees=45
  twistMotion=Free
 }
]
