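# Post-process style AnimBP for the Lemon Lime female body cosmetic, adding secondary motion to jacket, collar, belts and bandanna over the copied parent pose.
# Crouch and limb-swing bone correctives plus constraints feed a cached pose, driving two RigidBody setups blended by backbling presence.
# AnimDynamics and aim- or crouch-driven collar offsets finish the pose.
# EventGraph derives swing alphas, weapon flags and backbling state, and resets dynamics outside the front end.
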AnimBlueprint /BRCosmetics/Characters/Player/Female/Medium/Bodies/F_MED_Lemon_Lime/Meshes/Lemon_Lime_AnimBP
target_skeleton=&/Lemon_Lime_Skeleton
graphs [
 AnimGraph {nodes [
   Blend Poses by bool(bActiveValue=Get isPistol() -> (isPistol:bool)) {blendType=Linear}
   "Constraint - dyn_jacketBelt01_Base_r (PPPPPPPPPP)"!(
    ComponentPose
    Alpha=1
    ConstraintWeights_0=0.05
    ConstraintWeights_1:float=0.1933
    ConstraintWeights_2:float=0.001
    ConstraintWeights_3:float=0
    ConstraintWeights_4:float=0.0338
    ConstraintWeights_5:float=0.0104
    ConstraintWeights_6:float=0.0003
    ConstraintWeights_7:float=0.0085
    ConstraintWeights_8:float=0
    ConstraintWeights_9:float=0.7027
   ) -> (Pose) {
    alphaBoolBlend {blendOption=Linear}
    boneToModify {boneName=dyn_jacketBelt01_Base_r}
    constraintSetup [
     {targetBone {boneName=dyn_jacketBelt01_Base_r}}
     {targetBone {boneName=dyn_jacketSideA01_r}}
     {targetBone {boneName=upperarm_twist_01_r}}
     {targetBone {boneName=dyn_jacketBack_02_l}}
     {targetBone {boneName=spine_04}}
     {targetBone {boneName=spine_05}}
     {targetBone {boneName=dyn_jacketSideB_03_r}}
     {targetBone {boneName=clavicle_r}}
     {targetBone {boneName=dyn_jacketBack_02_r}}
     {targetBone {boneName=dyn_jacketFront01_r}}
    ]
    constraintWeights [1,1,1,1,1,1,1,1,1,1]
   }
   "Constraint - dyn_jacketBelt01_Base_r (PPPPPPP)"!(
    ComponentPose
    Alpha=1
    ConstraintWeights_0=0.05
    ConstraintWeights_1:float=0.1933
    ConstraintWeights_2:float=0.0009
    ConstraintWeights_3:float=0.703
    ConstraintWeights_4:float=0.0339
    ConstraintWeights_5:float=0.0104
    ConstraintWeights_6:float=0.0085
   ) -> (Pose) {
    alphaBoolBlend {blendOption=Linear}
    boneToModify {boneName=dyn_jacketBelt01_Base_r}
    constraintSetup [
     {targetBone {boneName=dyn_jacketBelt01_Base_r}}
     {targetBone {boneName=dyn_jacketSideA01_r}}
     {targetBone {boneName=upperarm_twist_01_r}}
     {targetBone {boneName=dyn_jacketFront01_r}}
     {targetBone {boneName=spine_04}}
     {targetBone {boneName=spine_05}}
     {targetBone {boneName=clavicle_r}}
    ]
    constraintWeights [1,1,1,1,1,1,1]
   }
   "Constraint - dyn_jacketBelt01_Base_r (PPPPPPP)"(
    Alpha=1
    ConstraintWeights_0=0.0389
    ConstraintWeights_1=0.1097
    ConstraintWeights_2=0.0007
    ConstraintWeights_3=0.7986
    ConstraintWeights_4=0.0286
    ConstraintWeights_5=0.0126
    ConstraintWeights_6=0.0109
   ) {
    alphaBoolBlend {blendOption=Linear}
    boneToModify {boneName=dyn_jacketBelt01_Base_r}
    constraintSetup [
     {targetBone {boneName=dyn_jacketBelt01_Base_r}}
     {targetBone {boneName=dyn_jacketSideA01_r}}
     {targetBone {boneName=upperarm_twist_01_r}}
     {targetBone {boneName=dyn_jacketFront01_r}}
     {targetBone {boneName=spine_04}}
     {targetBone {boneName=spine_05}}
     {targetBone {boneName=clavicle_r}}
    ]
    constraintWeights [1,1,1,1,1,1,1]
   }
   "Transform (Modify) Bone - Bone: dyn_collarD_01_r"!(ComponentPose, bAlphaBoolEnabled:bool=True, Translation="0,20,0", Rotation="0,-120,0", Scale="1,1,1") -> (Pose) {
    alphaInputType=Bool
    rotationMode=BMM_Additive
    rotationSpace=BCS_BoneSpace
    alphaBoolBlend {blendInTime=0.1, blendOption=Linear, blendOutTime=0.5}
    boneToModify {boneName=dyn_collarD_01_r}
    rotation {yaw=-120}
    translation {y=20}
   }
   "Transform (Modify) Bone - Bone: dyn_collarBack_01"!(ComponentPose, bAlphaBoolEnabled:bool=True, Translation="0,20,0", Rotation="0,-90,0", Scale="1,1,1") -> (Pose) {
    alphaInputType=Bool
    rotationMode=BMM_Additive
    rotationSpace=BCS_BoneSpace
    alphaBoolBlend {blendInTime=0.1, blendOption=Linear, blendOutTime=0.5}
    boneToModify {boneName=dyn_collarBack_01}
    rotation {yaw=-90}
    translation {y=20}
   }
   n1 = AND Boolean!(
    A=OR Boolean(A=Get isAimingRifle() -> (isAimingRifle:bool), B=Get isPistol())
    B=Get bIsCrouching()
    C:bool=Get bIsAccelerating2D()
    D:bool=true
   ) -> (ReturnValue)
   n2 = "Transform (Modify) Bone - Bone: dyn_collarD_01_l"!(ComponentPose, bAlphaBoolEnabled:bool=@n1, Translation="0,0,0", Rotation="0,0,0", Scale="1,1,1") -> (Pose) {
    alphaInputType=Bool
    rotationMode=BMM_Additive
    rotationSpace=BCS_BoneSpace
    alphaBoolBlend {blendInTime=0.5, blendOption=Linear, blendOutTime=1}
    boneToModify {boneName=dyn_collarD_01_l}
   }
   "Transform (Modify) Bone - Bone: dyn_collarB_01_l"!(ComponentPose=@n2, bAlphaBoolEnabled:bool=@n1, Translation="0,0,0", Rotation="0,0,0", Scale="1,1,1") -> (Pose) {
    alphaInputType=Bool
    rotationMode=BMM_Additive
    rotationSpace=BCS_BoneSpace
    alphaBoolBlend {blendInTime=0.5, blendOption=Linear, blendOutTime=1}
    boneToModify {boneName=dyn_collarB_01_l}
   }
   RigidBody() {
    bClampLinearTranslationLimitToRefPose=true
    bOverrideWorldGravity=true
    overridePhysicsAsset=&/Lemon_Lime_Backpack_Physics
    simulationSpace=BaseBoneSpace
    alphaBoolBlend {blendOption=Linear}
    baseBoneRef {boneName=root}
    componentAppliedLinearAccClamp=<2000 2000 2000>
    componentLinearAccScale=<0.25 0.6 0.6>
    componentLinearVelScale=<0.5 0.4 0.4>
    overrideWorldGravity {z=-980}
   }
   n3 = RigidBody(ComponentPose=Local To Component(LocalPose=Use cached pose 'PreRigid'()), Alpha=SelectFloat(A=0.9, B=0.75, bPickA=Get bIsSkyDiving())) {
    bClampLinearTranslationLimitToRefPose=true
    bOverrideWorldGravity=true
    overridePhysicsAsset=&/Lemon_Lime_Physics
    simulationSpace=BaseBoneSpace
    alphaBoolBlend {blendOption=Linear}
    baseBoneRef {boneName=spine_03}
    componentAppliedLinearAccClamp=<2000 2000 2000>
    componentLinearAccScale=<0.25 0.3 0.3>
    componentLinearVelScale=<0.5 0.2 0.2>
    overrideWorldGravity {z=-980}
   }
   n4 = Component To Local(ComponentPose=@n3)
   n5 = RigidBody(ComponentPose=Local To Component(LocalPose=Use cached pose 'PreRigid'())) {
    bClampLinearTranslationLimitToRefPose=true
    bOverrideWorldGravity=true
    overridePhysicsAsset=&Lemon_Lime_Physics
    simulationSpace=BaseBoneSpace
    alphaBoolBlend {blendOption=Linear}
    baseBoneRef {boneName=spine_03}
    componentAppliedLinearAccClamp=<2000 2000 2000>
    componentLinearAccScale=<0.25 0.6 0.6>
    componentLinearVelScale=<0.5 0.4 0.4>
    overrideWorldGravity {z=-980}
   }
   n6 = Component To Local(ComponentPose=@n5)
   n7 = Get capeAlpha() -> (capeAlpha:bool)
   n8 = "Constraint - dyn_jacketBack_02_l (R)"(
    ComponentPose=Local To Component(LocalPose=Blend Poses by bool(BlendPose_0=@n6, BlendPose_1=@n4, bActiveValue=@n7) {blendType=Linear})
    Alpha=0.5
    ConstraintWeights_0=1
   ) {
    alphaBoolBlend {blendOption=Linear}
    boneToModify {boneName=dyn_jacketBack_02_l}
    constraintSetup [{transformType=Rotation, targetBone {boneName=dyn_jacketBack_02}}]
    constraintWeights [1]
   }
   n9 = "Constraint - dyn_jacketBack_02_r (R)"(ComponentPose=@n8, Alpha=0.5, ConstraintWeights_0=1) {
    alphaBoolBlend {blendOption=Linear}
    boneToModify {boneName=dyn_jacketBack_02_r}
    constraintSetup [{transformType=Rotation, targetBone {boneName=dyn_jacketBack_02}}]
    constraintWeights [1]
   }
   n10 = "Constraint - dyn_jacketBelt01_Base_r (PPPPPPP)"(
    ComponentPose=@n9
    Alpha=1
    ConstraintWeights_0=0.0389
    ConstraintWeights_1=0.1097
    ConstraintWeights_2=0.0007
    ConstraintWeights_3=0.7986
    ConstraintWeights_4=0.0286
    ConstraintWeights_5=0.0126
    ConstraintWeights_6=0.0109
   ) {
    alphaBoolBlend {blendOption=Linear}
    boneToModify {boneName=dyn_jacketBelt01_Base_r}
    constraintSetup [
     {targetBone {boneName=dyn_jacketBelt01_Base_r}}
     {targetBone {boneName=dyn_jacketSideA01_r}}
     {targetBone {boneName=upperarm_twist_01_r}}
     {targetBone {boneName=dyn_jacketFront01_r}}
     {targetBone {boneName=spine_04}}
     {targetBone {boneName=spine_05}}
     {targetBone {boneName=clavicle_r}}
    ]
    constraintWeights [1,1,1,1,1,1,1]
   }
   n11 = "Constraint - dyn_jacketBelt01_r (P)"(ComponentPose=@n10, Alpha=1, ConstraintWeights_0=1) {
    alphaBoolBlend {blendOption=Linear}
    boneToModify {boneName=dyn_jacketBelt01_r}
    constraintSetup [{targetBone {boneName=dyn_jacketFront01_r}}]
    constraintWeights [1]
   }
   n12 = Anim Dynamics - Bone: dyn_jacketBelt01_l(ComponentPose=@n11, Alpha=1, ExternalForce="0,0,0") {
    angularDampingOverride=0.9999
    bGravityOverrideInSimSpace=true
    bOverrideAngularDamping=true
    bUseGravityOverride=true
    simulationSpace=RootRelative
    alphaBoolBlend {blendOption=Linear}
    boundBone {boneName=dyn_jacketBelt01_l}
    componentLinearAccScale=<0.1 0.1 0.1>
    componentLinearVelScale=<0.1 0.1 0.1>
    gravityOverride {z=-980}
    physicsBodyDefinitions [{
      sphereCollisionRadius=0
      boundBone {boneName=dyn_jacketBelt01_l}
      boxExtents=<3 5 5>
      constraintSetup {angularLimitsMax=<5 10 0>, angularLimitsMin=<-5 -10 -10>, angularTarget=<0 0 0>}
      localJointOffset=<1 0 0>
     }]
   }
   n13 = Anim Dynamics - Bone: dyn_jacketBelt02_l(ComponentPose=@n12, Alpha=1, ExternalForce="0,0,0") {
    angularDampingOverride=0.9999
    bGravityOverrideInSimSpace=true
    bOverrideAngularDamping=true
    bUseGravityOverride=true
    simulationSpace=RootRelative
    alphaBoolBlend {blendOption=Linear}
    boundBone {boneName=dyn_jacketBelt02_l}
    componentLinearAccScale=<0.1 0.1 0.1>
    componentLinearVelScale=<0.1 0.1 0.1>
    gravityOverride {z=-980}
    physicsBodyDefinitions [{
      sphereCollisionRadius=0
      boundBone {boneName=dyn_jacketBelt02_l}
      boxExtents=<3 5 5>
      constraintSetup {angularLimitsMax=<5 10 0>, angularLimitsMin=<-5 -10 -10>, angularTarget=<0 0 0>}
      localJointOffset=<1 0 0>
     }]
   }
   n14 = Anim Dynamics - Bone: dyn_jacketBelt01_r(ComponentPose=@n13, Alpha=1, ExternalForce="0,0,0") {
    angularDampingOverride=0.999
    bGravityOverrideInSimSpace=true
    bOverrideAngularDamping=true
    bUseGravityOverride=true
    simulationSpace=RootRelative
    alphaBoolBlend {blendOption=Linear}
    boundBone {boneName=dyn_jacketBelt01_r}
    componentLinearAccScale=<0.1 0.1 0.1>
    componentLinearVelScale=<0.1 0.1 0.1>
    gravityOverride {z=-980}
    physicsBodyDefinitions [{
      sphereCollisionRadius=0
      boundBone {boneName=dyn_jacketBelt01_r}
      boxExtents=<3 5 5>
      constraintSetup {angularLimitsMax=<5 10 0>, angularLimitsMin=<-5 -10 -10>, angularTarget=<0 0 0>}
      localJointOffset=<1 0 0>
     }]
   }
   n15 = Anim Dynamics - Bone: dyn_jacketBelt02_r(ComponentPose=@n14, Alpha=1, ExternalForce="0,0,0") {
    angularDampingOverride=0.999
    bGravityOverrideInSimSpace=true
    bOverrideAngularDamping=true
    bUseGravityOverride=true
    simulationSpace=RootRelative
    alphaBoolBlend {blendOption=Linear}
    boundBone {boneName=dyn_jacketBelt02_r}
    componentLinearAccScale=<0.1 0.1 0.1>
    componentLinearVelScale=<0.1 0.1 0.1>
    gravityOverride {z=-980}
    physicsBodyDefinitions [{
      sphereCollisionRadius=0
      boundBone {boneName=dyn_jacketBelt02_r}
      boxExtents=<3 5 5>
      constraintSetup {angularLimitsMax=<5 10 0>, angularLimitsMin=<-5 -10 -10>, angularTarget=<0 0 0>}
      localJointOffset=<1 0 0>
     }]
   }
   n16 = Anim Dynamics - Bone: dyn_bandannaA(ComponentPose=@n15, Alpha=1, ExternalForce="0,0,0") {
    angularDampingOverride=0.999
    bGravityOverrideInSimSpace=true
    bOverrideAngularDamping=true
    bUseGravityOverride=true
    simulationSpace=RootRelative
    alphaBoolBlend {blendOption=Linear}
    boundBone {boneName=dyn_bandannaA}
    componentLinearAccScale=<0.5 0.5 0.5>
    componentLinearVelScale=<0.5 0.5 0.5>
    gravityOverride {z=-980}
    physicsBodyDefinitions [{
      sphereCollisionRadius=0
      boundBone {boneName=dyn_bandannaA}
      boxExtents=<2 10 10>
      constraintSetup {angularLimitsMax=<7.5 7.5 7.5>, angularLimitsMin=<-7.5 -7.5 -7.5>, angularTarget=<0 0 0>}
      localJointOffset=<3 0 0>
     }]
   }
   n17 = Anim Dynamics - Bone: dyn_bandannaB(ComponentPose=@n16, Alpha=1, ExternalForce="0,0,0") {
    angularDampingOverride=0.999
    bGravityOverrideInSimSpace=true
    bOverrideAngularDamping=true
    bUseGravityOverride=true
    simulationSpace=RootRelative
    alphaBoolBlend {blendOption=Linear}
    boundBone {boneName=dyn_bandannaB}
    componentLinearAccScale=<0.5 0.5 0.5>
    componentLinearVelScale=<0.5 0.5 0.5>
    gravityOverride {z=-980}
    physicsBodyDefinitions [{
      sphereCollisionRadius=0
      boundBone {boneName=dyn_bandannaB}
      boxExtents=<2 10 10>
      constraintSetup {angularLimitsMax=<7.5 7.5 7.5>, angularLimitsMin=<-7.5 -7.5 -7.5>, angularTarget=<0 0 0>}
      localJointOffset=<3 0 0>
     }]
   }
   n18 = Anim Dynamics - Bone: dyn_collarStrap_01_l(ComponentPose=@n17, Alpha=1, ExternalForce="0,0,0") {
    angularDampingOverride=0.999
    bGravityOverrideInSimSpace=true
    bOverrideAngularDamping=true
    bUseGravityOverride=true
    simulationSpace=RootRelative
    alphaBoolBlend {blendOption=Linear}
    boundBone {boneName=dyn_collarStrap_01_l}
    componentLinearAccScale=<0.5 0.5 0.5>
    componentLinearVelScale=<0.5 0.5 0.5>
    gravityOverride {z=-980}
    physicsBodyDefinitions [{
      sphereCollisionRadius=0
      boundBone {boneName=dyn_collarStrap_01_l}
      boxExtents=<2 10 10>
      constraintSetup {angularLimitsMax=<7.5 7.5 7.5>, angularLimitsMin=<-7.5 -7.5 -7.5>, angularTarget=<0 0 0>}
      localJointOffset=<3 0 0>
     }]
   }
   n19 = Anim Dynamics - Bone: dyn_collarStrap_01_r(ComponentPose=@n18, Alpha=1, ExternalForce="0,0,0") {
    angularDampingOverride=0.999
    bGravityOverrideInSimSpace=true
    bOverrideAngularDamping=true
    bUseGravityOverride=true
    simulationSpace=RootRelative
    alphaBoolBlend {blendOption=Linear}
    boundBone {boneName=dyn_collarStrap_01_r}
    componentLinearAccScale=<0.5 0.5 0.5>
    componentLinearVelScale=<0.5 0.5 0.5>
    gravityOverride {z=-980}
    physicsBodyDefinitions [{
      sphereCollisionRadius=0
      boundBone {boneName=dyn_collarStrap_01_r}
      boxExtents=<2 10 10>
      constraintSetup {angularLimitsMax=<7.5 7.5 7.5>, angularLimitsMin=<-7.5 -7.5 -7.5>, angularTarget=<0 0 0>}
      localJointOffset=<3 0 0>
     }]
   }
   n20 = "Transform (Modify) Bone - Bone: dyn_collarA_01_r"!(ComponentPose=@n19, bAlphaBoolEnabled:bool=Get isAimingRifle(), Translation="0,0,0", Rotation="10,0,0", Scale="1,1,1") -> (Pose) {
    alphaInputType=Bool
    rotationMode=BMM_Additive
    rotationSpace=BCS_BoneSpace
    alphaBoolBlend {blendInTime=0.1, blendOption=Linear, blendOutTime=0.5}
    boneToModify {boneName=dyn_collarA_01_r}
    rotation {pitch=10}
   }
   n21 = "Transform (Modify) Bone - Bone: dyn_collarB_01_r"!(ComponentPose=@n20, bAlphaBoolEnabled:bool=Get isAimingRifle(), Translation="0,0,0", Rotation="20,0,10", Scale="1,1,1") -> (Pose) {
    alphaInputType=Bool
    rotationMode=BMM_Additive
    rotationSpace=BCS_BoneSpace
    alphaBoolBlend {blendInTime=0.1, blendOption=Linear, blendOutTime=0.5}
    boneToModify {boneName=dyn_collarB_01_r}
    rotation {pitch=20, roll=10}
   }
   n22 = "Transform (Modify) Bone - Bone: dyn_collarD_01_r"(ComponentPose=@n21, bAlphaBoolEnabled=Get isAimingRifle(), Translation="0,0,0", Rotation="20,0,10", Scale="1,1,1") {
    alphaInputType=Bool
    rotationMode=BMM_Additive
    rotationSpace=BCS_BoneSpace
    alphaBoolBlend {blendInTime=0.1, blendOption=Linear, blendOutTime=0.5}
    boneToModify {boneName=dyn_collarD_01_r}
    rotation {pitch=20, roll=10}
   }
   n23 = "Transform (Modify) Bone - Bone: dyn_collarBack_01"(ComponentPose=@n22, bAlphaBoolEnabled=Get isAimingRifle(), Translation="0,0,0", Rotation="20,0,0", Scale="1,1,1") {
    alphaInputType=Bool
    rotationMode=BMM_Additive
    rotationSpace=BCS_BoneSpace
    alphaBoolBlend {blendInTime=0.1, blendOption=Linear, blendOutTime=0.5}
    boneToModify {boneName=dyn_collarBack_01}
    rotation {pitch=20}
   }
   n24 = "Transform (Modify) Bone - Bone: dyn_collarA_01_r"(ComponentPose=@n23, bAlphaBoolEnabled=@n1, Translation="0,0,0", Rotation="-50,0,20", Scale="1,1,1") {
    alphaInputType=Bool
    rotationMode=BMM_Additive
    rotationSpace=BCS_BoneSpace
    alphaBoolBlend {blendInTime=0.5, blendOption=Linear, blendOutTime=1}
    boneToModify {boneName=dyn_collarA_01_r}
    rotation {pitch=-50, roll=20}
   }
   n25 = "Transform (Modify) Bone - Bone: dyn_collarB_01_r"(ComponentPose=@n24, bAlphaBoolEnabled=@n1, Translation="0,0,0", Rotation="-50,0,0", Scale="1,1,1") {
    alphaInputType=Bool
    rotationMode=BMM_Additive
    rotationSpace=BCS_BoneSpace
    alphaBoolBlend {blendInTime=0.5, blendOption=Linear, blendOutTime=1}
    boneToModify {boneName=dyn_collarB_01_r}
    rotation {pitch=-50}
   }
   n26 = "Transform (Modify) Bone - Bone: dyn_collarD_01_r"(ComponentPose=@n25, bAlphaBoolEnabled=@n1, Translation="0,0,0", Rotation="0,-110,0", Scale="1,1,1") {
    alphaInputType=Bool
    rotationMode=BMM_Additive
    rotationSpace=BCS_BoneSpace
    alphaBoolBlend {blendInTime=0.5, blendOption=Linear, blendOutTime=1}
    boneToModify {boneName=dyn_collarD_01_r}
    rotation {yaw=-110}
   }
   n27 = "Transform (Modify) Bone - Bone: dyn_collarBack_01"(ComponentPose=@n26, bAlphaBoolEnabled=@n1, Translation="0,20,0", Rotation="0,-70,0", Scale="1,1,1") {
    alphaInputType=Bool
    rotationMode=BMM_Additive
    rotationSpace=BCS_BoneSpace
    alphaBoolBlend {blendInTime=0.5, blendOption=Linear, blendOutTime=1}
    boneToModify {boneName=dyn_collarBack_01}
    rotation {yaw=-70}
    translation {y=20}
   }
   n28 = "Constraint - dyn_collarA_01_r (R)"(ComponentPose=@n27, Alpha=0.35, ConstraintWeights_0=1) {
    alphaBoolBlend {blendOption=Linear}
    boneToModify {boneName=dyn_collarA_01_r}
    constraintSetup [{transformType=Rotation, targetBone {boneName=upperarm_r}}]
    constraintWeights [1]
   }
   n29 = "Constraint - dyn_collarB_01_r (R)"(ComponentPose=@n28, Alpha=0.4, ConstraintWeights_0=1) {
    alphaBoolBlend {blendOption=Linear}
    boneToModify {boneName=dyn_collarB_01_r}
    constraintSetup [{transformType=Rotation, targetBone {boneName=upperarm_r}}]
    constraintWeights [1]
   }
   n30 = "Constraint - dyn_collarD_01_l (R)"(ComponentPose=@n29, Alpha=0.4, ConstraintWeights_0=1) {
    alphaBoolBlend {blendOption=Linear}
    boneToModify {boneName=dyn_collarD_01_l}
    constraintSetup [{transformType=Rotation, targetBone {boneName=upperarm_r}}]
    constraintWeights [1]
   }
   n31 = "Constraint - dyn_collarBack_01 (R)"(ComponentPose=@n30, Alpha=0.1, ConstraintWeights_0=1) {
    alphaBoolBlend {blendOption=Linear}
    boneToModify {boneName=dyn_collarBack_01}
    constraintSetup [{transformType=Rotation, targetBone {boneName=upperarm_r}}]
    constraintWeights [1]
   }
   n32 = "Constraint - dyn_collarStrap_01_r (R)"(ComponentPose=@n31, Alpha=0.1, ConstraintWeights_0=1) {
    alphaBoolBlend {blendOption=Linear}
    boneToModify {boneName=dyn_collarStrap_01_r}
    constraintSetup [{transformType=Rotation, targetBone {boneName=upperarm_r}}]
    constraintWeights [1]
   }
   Output Pose(Result=Component To Local(ComponentPose=@n32)) {layerGroup=None}
   n33 = "To Float (Boolean)"(InBool=Get bIsCrouching())
   n34 = "Transform (Modify) Bone - Bone: dyn_jacketBack_01_Base"(
    ComponentPose=Local To Component(LocalPose=Copy Pose From Mesh() {bUseAttachedParent=true})
    Alpha=@n33
    Translation="0,-2,1"
    Rotation="0,0,0"
    Scale="1,1,1"
   ) {
    scaleMode=BMM_Replace
    scaleSpace=BCS_BoneSpace
    translationMode=BMM_Additive
    translationSpace=BCS_BoneSpace
    alphaBoolBlend {blendOption=Linear}
    boneToModify {boneName=dyn_jacketBack_01_Base}
    translation {y=-2, z=1}
   }
   n35 = "Transform (Modify) Bone - Bone: dyn_jacketBack_01_Base_l"(ComponentPose=@n34, Alpha=@n33, Translation="-0.000,-2,1.000", Rotation="0,0,0", Scale="1,1,1") {
    scaleMode=BMM_Replace
    scaleSpace=BCS_BoneSpace
    translationMode=BMM_Additive
    translationSpace=BCS_BoneSpace
    alphaBoolBlend {blendOption=Linear}
    boneToModify {boneName=dyn_jacketBack_01_Base_l}
    translation {y=-2, z=1}
   }
   n36 = "Transform (Modify) Bone - Bone: dyn_jacketBack_01_Base_r"(ComponentPose=@n35, Alpha=@n33, Translation="0,-2,1", Rotation="0,0,0", Scale="1,1,1") {
    scaleMode=BMM_Replace
    scaleSpace=BCS_BoneSpace
    translationMode=BMM_Additive
    translationSpace=BCS_BoneSpace
    alphaBoolBlend {blendOption=Linear}
    boneToModify {boneName=dyn_jacketBack_01_Base_r}
    translation {y=-2, z=1}
   }
   n37 = Get armSwingDelta_r() -> (armSwingDelta_r:double)
   n38 = "Transform (Modify) Bone - Bone: dyn_collarCorner_Base_r"(ComponentPose=@n36, Alpha=@n37, Translation="-0.5,2,-0.000", Rotation="0.000,15,-16", Scale="1,1,1") {
    rotationMode=BMM_Additive
    rotationSpace=BCS_BoneSpace
    scaleMode=BMM_Replace
    scaleSpace=BCS_BoneSpace
    translationMode=BMM_Additive
    translationSpace=BCS_BoneSpace
    alphaBoolBlend {blendOption=Linear}
    alphaScaleBiasClamp {bMapRange=true, inRange {max=0.65}}
    boneToModify {boneName=dyn_collarCorner_Base_r}
    rotation {roll=-16, yaw=15}
    translation=<-0.5 2>
   }
   n39 = "Transform (Modify) Bone - Bone: dyn_collarA_01_Base_r"(ComponentPose=@n38, Alpha=@n37, Translation="0.3,-0.000,-1.4", Rotation="-8,0.000,0.000", Scale="1,1,1") {
    rotationMode=BMM_Additive
    rotationSpace=BCS_BoneSpace
    scaleMode=BMM_Replace
    scaleSpace=BCS_BoneSpace
    translationMode=BMM_Additive
    translationSpace=BCS_BoneSpace
    alphaBoolBlend {blendOption=Linear}
    boneToModify {boneName=dyn_collarA_01_Base_r}
    rotation {pitch=-8}
    translation {x=0.3, z=-1.4}
   }
   n40 = "Transform (Modify) Bone - Bone: thigh_scale_r"(ComponentPose=@n39, Alpha=Get legSwingDelta_r() -> (legSwingDelta_r:double), Translation="10,-10,0", Rotation="0,0,0", Scale="14,14,14") {
    scaleMode=BMM_Replace
    scaleSpace=BCS_BoneSpace
    translationMode=BMM_Additive
    translationSpace=BCS_BoneSpace
    alphaBoolBlend {blendOption=Linear}
    boneToModify {boneName=thigh_scale_r}
    scale=<14 14 14>
    translation=<10 -10>
   }
   n41 = "Transform (Modify) Bone - Bone: thigh_scale_l"(ComponentPose=@n40, Alpha=Get legSwingDelta_l() -> (legSwingDelta_l:double), Translation="10,10,0", Rotation="0,0,0", Scale="10,10,10") {
    scaleMode=BMM_Replace
    scaleSpace=BCS_BoneSpace
    translationMode=BMM_Additive
    translationSpace=BCS_BoneSpace
    alphaBoolBlend {blendOption=Linear}
    boneToModify {boneName=thigh_scale_l}
    scale=<10 10 10>
    translation=<10 10>
   }
   n42 = "Constraint - dyn_bandanna (PPP)"!(ComponentPose=@n41, Alpha=1, ConstraintWeights_0=0.1429, ConstraintWeights_1:float=0.126, ConstraintWeights_2:float=0.7311) -> (Pose) {
    alphaBoolBlend {blendOption=Linear}
    boneToModify {boneName=dyn_bandanna}
    constraintSetup [{targetBone {boneName=upperarm_twist_02_l}}, {targetBone {boneName=upperarm_l}}, {targetBone {boneName=upperarm_twist_01_l}}]
    constraintWeights [1,1,1]
   }
   n43 = "Constraint - dyn_jacketBelt01_Base_l (PPPPPP)"!(
    ComponentPose=@n42
    Alpha=1
    ConstraintWeights_0=0.0167
    ConstraintWeights_1:float=1.8948
    ConstraintWeights_2:float=0.0218
    ConstraintWeights_3:float=0.0023
    ConstraintWeights_4:float=0.1128
    ConstraintWeights_5:float=0.5448
   ) -> (Pose) {
    alphaBoolBlend {blendOption=Linear}
    boneToModify {boneName=dyn_jacketBelt01_Base_l}
    constraintSetup [
     {targetBone {boneName=clavicle_l}}
     {targetBone {boneName=dyn_jacketFront01_l}}
     {targetBone {boneName=spine_05}}
     {targetBone {boneName=upperarm_twist_01_l}}
     {targetBone {boneName=spine_04}}
     {targetBone {boneName=dyn_jacketSideA01_l}}
    ]
    constraintWeights [1,1,1,1,1,1]
   }
   n44 = "Constraint - dyn_collarA_01_Base_l (PPP)"!(ComponentPose=@n43, Alpha=1, ConstraintWeights_0=0.6, ConstraintWeights_1:float=0.1, ConstraintWeights_2:float=0.3) -> (Pose) {
    alphaBoolBlend {blendOption=Linear}
    boneToModify {boneName=dyn_collarA_01_Base_l}
    constraintSetup [{targetBone {boneName=clavicle_l}}, {targetBone {boneName=upperarm_l}}, {targetBone {boneName=spine_05}}]
    constraintWeights [1,1,1]
   }
   n45 = "Constraint - dyn_collarB_01_Base_l (PPP)"!(ComponentPose=@n44, Alpha=1, ConstraintWeights_0=0.6, ConstraintWeights_1:float=0.1, ConstraintWeights_2:float=0.3) -> (Pose) {
    alphaBoolBlend {blendOption=Linear}
    boneToModify {boneName=dyn_collarB_01_Base_l}
    constraintSetup [{targetBone {boneName=clavicle_l}}, {targetBone {boneName=upperarm_l}}, {targetBone {boneName=spine_05}}]
    constraintWeights [1,1,1]
   }
   n46 = "Constraint - dyn_bandanna_Base (PPP)"!(ComponentPose=@n45, Alpha=1, ConstraintWeights_0=0.1429, ConstraintWeights_1:float=0.126, ConstraintWeights_2:float=0.7311) -> (Pose) {
    alphaBoolBlend {blendOption=Linear}
    boneToModify {boneName=dyn_bandanna_Base}
    constraintSetup [{targetBone {boneName=upperarm_twist_02_l}}, {targetBone {boneName=upperarm_l}}, {targetBone {boneName=upperarm_twist_01_l}}]
    constraintWeights [1,1,1]
   }
   Save cached pose 'PreRigid'(Pose=Component To Local(ComponentPose=@n46)) {cacheName=PreRigid}
   Get bIsCrouchSprinting()
   Get bIsTargeting()
  ]}
 EventGraph {nodes [
   Event BlueprintInitializeAnimation()
   n1 = TryGetPawnOwner()
   > Is Valid(InputObject=@n1)
   > n2 = Cast To PlayerPawn_Athena(Object=@n1)
   > Set PawnOwner(PawnOwner:Player Pawn Athena Object Reference=@n2) -> (Output_Get:Player Pawn Athena Object Reference)
   Event BlueprintUpdateAnimation()
   > n3 = Sequence!(execute) -> (then_0, then_1, then_2:Exec, then_3:Exec, then_4:Exec)
   n4 = Cast To FortPlayerAnimInstance(execute=@n3.then_3, Object=GetAnimInstance(self=GetOwnerSkeletalMeshComponent()))
   > Set isPistol(isPistol=false)
   > Set isAimingRifle(isAimingRifle:bool=false) -> (Output_Get:bool)
   > n5 = Switch on EFortWeaponCoreAnimation!(execute, Selection:EFortWeaponCoreAnimation Enum=Get WeaponCoreAnim(self=@n4)) -> (
    Melee:Exec
    Pistol:Exec
    Shotgun:Exec
    PaperBlueprint:Exec
    Rifle:Exec
    MeleeOneHand:Exec
    MachinePistol:Exec
    RocketLauncher:Exec
    GrenadeLauncher:Exec
    GoingCommando:Exec
    AssaultRifle:Exec
    TacticalShotgun:Exec
    SniperRifle:Exec
    TrapPlacement:Exec
    ShoulderLauncher:Exec
    AbilityDecoTool:Exec
    Crossbow:Exec
    C4:Exec
    RemoteControl:Exec
    DualWield:Exec
    AR_BullPup:Exec
    AR_ForwardGrip:Exec
    MedPackPaddles:Exec
    SMG_P90:Exec
    AR_DrumGun:Exec
    Consumable_Small:Exec
    Consumable_Large:Exec
    Balloon:Exec
    MountedTurret:Exec
    CreativeTool:Exec
    ExplosiveBow:Exec
    AshtonIndigo:Exec
    AshtonChicago:Exec
    MeleeDualWield:Exec
    Unarmed:Exec
   )
   Branch(execute=[@n5.AssaultRifle, @n5.Crossbow, @n5.GrenadeLauncher, @n5.Rifle, @n5.RocketLauncher, @n5.Shotgun, @n5.SniperRifle, @n5.TacticalShotgun])
   > Set isAimingRifle(isAimingRifle=true)
   Branch(execute=[@n5.MachinePistol, @n5.Pistol])
   > Set isPistol(isPistol=true)
   n6 = Branch(execute=@n3.then_4, Condition=Get bIsInFrontEnd())
   Do Once(execute=@n6.else, Reset=@n6.then, Start Closed=true)
   > Delay(Duration=0.2)
   > ResetDynamics(InTeleportType=TeleportPhysics)
   > ResetDynamics(InTeleportType=ResetPhysics)
   n7 = GetOwnerSkeletalMeshComponent()
   Set legSwingDelta_r(
    execute=@n3.then_0
    legSwingDelta_r:double=MapRangeClamped(Value=GetTwistAndSwingAngleOfDeltaRotationFromRefPose(self=@n7, BoneName=thigh_r).OutSwingAngle, InRangeA=173, InRangeB=72)
   ) -> (Output_Get:double)
   > Set legSwingDelta_l(legSwingDelta_l:double=MapRangeClamped(Value=GetTwistAndSwingAngleOfDeltaRotationFromRefPose(self=@n7, BoneName=thigh_l).OutSwingAngle, InRangeA=24, InRangeB=101)) -> (Output_Get:double)
   Do Once(execute=[@n13.then, @n3.then_1])
   > Is Valid(InputObject=Get PawnOwner() -> (PawnOwner:Player Pawn Athena Object Reference))
   n8 = CheckForBackbling()
   n9 = OR Boolean(A=@n8, B=@n8.bIsBackblingLong)
   > Set capeAlpha(capeAlpha:bool=@n9) -> (Output_Get:bool)
   n10 = "To Float (Boolean)"(InBool=@n9)
   > Set CapeAlphaInv(CapeAlphaInv:double=@n10) -> (Output_Get:double)
   n11 = BreakRotator(InRot=BreakTransform(InTransform=GetDeltaTransformFromRefPose(self=@n7, BoneName=upperarm_r, BaseName=clavicle_r)).Rotation)
   Set armSwingDelta_r(execute=@n3.then_2, armSwingDelta_r:double=MapRangeClamped(Value="Absolute (Float)"(A=@n11.Yaw), InRangeA=45, InRangeB=116)) -> (Output_Get:double)
   n12 = BreakRotator(InRot=BreakTransform(InTransform=GetDeltaTransformFromRefPose(self=@n7, BoneName=upperarm_l, BaseName=clavicle_l)).Rotation)
   > Set armSwingDelta_l(armSwingDelta_l:double=MapRangeClamped(Value="Absolute (Float)"(A=@n12.Yaw), InRangeA=5, InRangeB=70)) -> (Output_Get:double)
   n13 = Event OnCharacterCustomizationLoaded()
   TryGetPawnOwner()
   GetOwnerSkeletalMeshComponent()
   float - float(B=@n10) {is=title K2Node_CallFunction}
   n14 = GetSkeletalMeshForPartType(self=Get PawnOwner())
   "Not Equal (Object)"!(
    A:Object Reference=FindChildCommponentFromAttachSocket!(
     self:Custom Character Part Anim Instance Object Reference
     Parent:Skeletal Mesh Component Object Reference=@n14
     bIncludeAllDescendants:bool=false
     AttachSocket:Name=Backpack_BR
    ) -> (ReturnValue:Skeletal Mesh Component Object Reference)
    B:Object Reference
   ) -> (ReturnValue:bool)
   "Not Equal (Object)"(A=FindChildCommponentFromAttachSocket(Parent=@n14, bIncludeAllDescendants=false, AttachSocket=cape))
   Reroute Node(InputPin:Wildcard) -> (OutputPin:Wildcard)
   Set isAimingPistol(isAimingPistol:double=0) -> (Output_Get:double)
  ]}
]
variables [
 armSwingDelta_l {type=number}
 armSwingDelta_r {type=number}
 capeAlpha {type=boolean}
 CapeAlphaInv {type=number}
 isAimingPistol {type=number}
 isAimingRifle {type=boolean}
 isPistol {type=boolean}
 legSwingDelta_l {type=number}
 legSwingDelta_r {type=number}
 PawnOwner {type=PlayerPawn_Athena_C}
]
